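# Niagara system for the PillowMill Myriad character's pupil effect; its only emitter, EyeGlow, is disabled and has no renderers.
# EyeGlow bursts two particles, using spawn groups to place each on its own bone or socket of the system's skeletal mesh.
# Particles take System.ColorEye, with system-level dependencies including a color-switch helper.
# User parameters expose color variant, ADS flags and eye-flame toggle, amount and side.
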NiagaraSystem /BRCosmetics/Characters/Player/Female/Medium/Bodies/F_MED_PillowMill_Myriad/FX/NS_Character_PillowMill_Myriad_SL_Pupil
dependencies {
 usedDataInterfaces [&NiagaraDataInterfaceArrayInt32, &NiagaraDataInterfaceSkeletalMesh]
 usedDynamicInputs [&MergeHelper_ColorSwitch, &MakeIntFromFloat]
 usedModules [
  &self:SystemSpawnScript.SystemScriptSource.SystemScriptGraph.NiagaraNodeAssignment_3.SetVariables_#1
  &SystemState
  &self:SystemSpawnScript.SystemScriptSource.SystemScriptGraph.NiagaraNodeAssignment_1.SetVariables_#2
 ]
}
emitters {EyeGlow {
  bEnabled=false
  emitterUpdateScript [
   &EmitterState(
    Inactive Response=::NewEnumerator1
    Life Cycle Mode=::NewEnumerator1
    Loop Behavior=::NewEnumerator1
    Loop Duration Mode=::NewEnumerator1
    UseLoopDelay=false
   )
   &SpawnBurst_Instantaneous(Spawn Count=2, Use Spawn Probability=false, UseLoopCountLimit=true)
  ]
  particleSpawnScript [
   &FN_SetSpawnGroup(Number of Spawn Groups=2)
   &InitializeParticle(Color Mode=::NewEnumerator0, Mesh Scale Mode=::NewEnumerator3, UsePositionOffset=false, Write Material Random=true)
  ]
  particleUpdateScript [
   &ParticleState(Kill Particles When Lifetime Has Elapsed=false, Let Infinitely Lived Particles Die When Emitter Deactivates=true)
   &ScaleMeshSize(Scale Factor=&VectorFromFloat(Value=1))
   &Color(Color=$System.ColorEye)
   &SkeletalMeshLocation(
    Bone / Socket Index=&SelectIntFromArray(Direct Array Index=$Particles.SpawnGroup, Int Selection Array=DataInterface() {propertyValues {IntData [1,0]}})
    Bone Sampling Mode=::NewEnumerator4
    MaskSpawnGroup=false
    Skeletal Mesh=$System.Skeletal Mesh
   )
  ]
  rendererClasses []
  renderers []
 }}
summary {
 systemName=NS_Character_PillowMill_Myriad_SL_Pupil
 emitters [{bEnabled=false, emitterName=EyeGlow, rendererClasses []}]
 userVariables [
  User.bisADS {type=NiagaraBool}
  User.bIsADSOther {type=NiagaraBool}
  User.ColorVariant {type=NiagaraFloat, defaultValue=0, description=""}
  User.EyeFlame_On/Off {type=NiagaraBool, defaultValue=true}
  User.EyeFlameAmount {type=NiagaraInt32, defaultValue=1}
  User.EyeFlameSide_Left/Right {type=NiagaraBool, defaultValue=true}
 ]
}
