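# Physics asset for the PillowMill Myriad medium female player body cosmetic, built around secondary motion for its sweatshirt.
# Capsules cover torso, pelvis, arms and legs, while small BASE_ bodies link the garment to spine_02 and spine_03.
# Front, side and back dyn_sweatshirt chains hang from those anchors with mostly tight swing and twist limits.
# Free-rotation constraints cross-link some bottom segments into a hem, more completely on the left than the right.
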PhysicsAsset /BRCosmetics/Characters/Player/Female/Medium/Bodies/F_MED_PillowMill_Myriad/Meshes/F_MED_PillowMill_Myriad_Physics
bodies {
 BASE_sweatshirt_L_back [{shapeType=Capsule, length=0.505, radius=0.505, shapeName=BASE_sweatshirt_L_back_capsule, rotation=<rot 0 -90 0>}]
 BASE_sweatshirt_L_back_02 [{shapeType=Capsule, length=0.505, radius=0.505, shapeName=BASE_sweatshirt_L_back_02_capsule, rotation=<rot 0 -90 0>}]
 BASE_sweatshirt_L_front [{shapeType=Capsule, length=0.505, radius=0.505, shapeName=BASE_sweatshirt_L_front_capsule, rotation=<rot 0 -90 0>}]
 BASE_sweatshirt_L_side [{shapeType=Capsule, length=0.505, radius=0.505, shapeName=BASE_sweatshirt_L_side_capsule, rotation=<rot 0 -90 0>}]
 BASE_sweatshirt_M_back [{shapeType=Capsule, length=0.505, radius=0.505, shapeName=BASE_sweatshirt_M_back_capsule, rotation=<rot 0 -90 0>}]
 BASE_sweatshirt_M_back_02 [{shapeType=Capsule, length=0.505, radius=0.505, shapeName=BASE_sweatshirt_M_back_02_capsule, rotation=<rot 0 -90 0>}]
 BASE_sweatshirt_R_back [{shapeType=Capsule, length=0.505, radius=0.505, shapeName=BASE_sweatshirt_R_back_capsule, rotation=<rot 0 -90 0>}]
 BASE_sweatshirt_R_back_02 [{shapeType=Capsule, length=0.505, radius=0.505, shapeName=BASE_sweatshirt_R_back_02_capsule, rotation=<rot 0 -90 0>}]
 BASE_sweatshirt_R_front [{shapeType=Capsule, length=0.505, radius=0.505, shapeName=BASE_sweatshirt_R_front_capsule, rotation=<rot 0 -90 0>}]
 BASE_sweatshirt_R_side [{shapeType=Capsule, length=0.505, radius=0.505, shapeName=BASE_sweatshirt_R_side_capsule, rotation=<rot 0 -90 0>}]
 calf_l [{shapeType=Capsule, length=22.2583, radius=6.761, center=<-20.6416 -1.2916 -0.3294>, rotation=<rot 0 88.4509 -89.1755>}]
 calf_r [{shapeType=Capsule, length=22.2583, radius=6.761, center=<20.6416 1.2916 0.3295>, rotation=<rot 0 88.451 -89.1748>}]
 dyn_sweatshirt_L_back_02 [{
   shapeType=Capsule
   length=4.1788
   radius=5.4814
   shapeName=dyn_sweatshirt_L_back_02_capsule
   center=<4.4948 -1.9976 3.7415>
   rotation=<rot 0 0 -90>
  }]
 dyn_sweatshirt_L_back_03 [{
   shapeType=Capsule
   length=13.6013
   radius=5.3166
   shapeName=dyn_sweatshirt_L_back_03_capsule
   center=<9.032 0.3821 2.7133>
   rotation=<rot -0.0001 0 -85.4566>
  }]
 dyn_sweatshirt_L_front_01 [{
   shapeType=Capsule
   length=7.2557
   radius=4.7328
   shapeName=dyn_sweatshirt_L_front_01_capsule
   center=<5.5812 -0.8427 2.4577>
   rotation=<rot -6.0491 -3.9707 -89.2239>
  }]
 dyn_sweatshirt_L_front_02 [
  {
   shapeType=Capsule
   length=11.2994
   radius=3.4907
   shapeName=dyn_sweatshirt_L_front_02_capsule
   center=<6.4088 -0.317 0.8028>
   rotation=<rot 5.2106 176.7924 -85.2983>
  }
  {
   shapeType=Capsule
   length=11.2994
   radius=3.4907
   shapeName=dyn_sweatshirt_L_front_02_capsule
   center=<11.6763 -0.0147 0.7826>
   rotation=<rot 5.2106 176.7924 -85.2983>
  }
 ]
 dyn_sweatshirt_L_front_03 [
  {
   shapeType=Capsule
   length=12.3449
   radius=2.9645
   shapeName=dyn_sweatshirt_L_front_03_capsule
   center=<7.8234 -2.4297 -0.9385>
   rotation=<rot 5.4153 173.7538 -96.5378>
  }
  {
   shapeType=Capsule
   length=8.3377
   radius=1.8652
   shapeName=dyn_sweatshirt_L_front_03_capsule
   center=<13.3314 -0.7794 -2.0709>
   rotation=<rot 5.4153 173.7538 -109.2778>
  }
 ]
 dyn_sweatshirt_L_side_01 [{
   shapeType=Capsule
   length=4.7807
   radius=3.2885
   shapeName=dyn_sweatshirt_L_side_01_capsule
   center=<4.3827 -0.5398 1.5896>
   rotation=<rot 0 177.0229 -78.2604>
  }]
 dyn_sweatshirt_L_side_02 [{
   shapeType=Capsule
   length=6.0175
   radius=4.0804
   shapeName=dyn_sweatshirt_L_side_02_capsule
   center=<5.771 -1.2761 2.0178>
   rotation=<rot 0 169.5893 -77.0368>
  }]
 dyn_sweatshirt_L_side_03 [{
   shapeType=Capsule
   length=11.6526
   radius=7.2873
   shapeName=dyn_sweatshirt_L_side_03_capsule
   center=<8.3069 -2.0043 2.8155>
   rotation=<rot -0.4407 173.4238 -83.5171>
  }]
 dyn_sweatshirt_M_back_02 [{
   shapeType=Capsule
   length=11.4935
   radius=7.8255
   shapeName=dyn_sweatshirt_R_back_02_capsule
   center=<5.3166 -0.4555 5.717>
   rotation=<rot 0.849 179.0509 -87.9961>
  }]
 dyn_sweatshirt_M_back_03 [{
   shapeType=Capsule
   length=10.6313
   radius=8.0669
   shapeName=dyn_sweatshirt_R_back_03_capsule
   center=<9.5381 -0.0998 8.0423>
   rotation=<rot 0.0032 -179.1986 -89.7625>
  }]
 dyn_sweatshirt_R_back_02 [{
   shapeType=Capsule
   length=8.4622
   radius=6.0503
   shapeName=dyn_sweatshirt_R_back_02_capsule
   center=<3.5744 -0.9801 4.3586>
   rotation=<rot 0.849 179.0509 -79.2919>
  }]
 dyn_sweatshirt_R_back_03 [{
   shapeType=Capsule
   length=7.8972
   radius=5.9945
   shapeName=dyn_sweatshirt_R_back_03_capsule
   center=<7.2324 -1.6563 5.825>
   rotation=<rot 0 -178.458 -81.0583>
  }]
 dyn_sweatshirt_R_front_01 [{
   shapeType=Capsule
   length=6.6647
   radius=4.3913
   shapeName=dyn_sweatshirt_R_front_01_capsule
   center=<3.1534 0.0966 1.9451>
   rotation=<rot -0.0917 178.1615 -89.2902>
  }]
 dyn_sweatshirt_R_front_02 [{
   shapeType=Capsule
   length=5.5406
   radius=4.871
   shapeName=dyn_sweatshirt_R_front_02_capsule
   center=<4.8624 0.1499 2.4416>
   rotation=<rot 1.6258 2.7654 -89.9713>
  }]
 dyn_sweatshirt_R_front_03 [
  {
   shapeType=Capsule
   length=12.3449
   radius=2.9645
   shapeName=dyn_sweatshirt_L_front_03_capsule
   center=<7.14 2.1881 -0.2913>
   rotation=<rot 3.4646 -171.1003 -86.675>
  }
  {
   shapeType=Capsule
   length=8.3377
   radius=1.8652
   shapeName=dyn_sweatshirt_L_front_03_capsule
   center=<12.8584 1.4049 -2.2453>
   rotation=<rot 3.0008 -168.0152 -72.837>
  }
  {
   shapeType=Capsule
   length=4.1181
   radius=1.8652
   shapeName=dyn_sweatshirt_R_front_03_capsule
   center=<11.2033 6.2411 -2.2364>
   rotation=<rot 2.9978 -168.0174 -100.6625>
  }
 ]
 dyn_sweatshirt_R_side_01 [{
   shapeType=Capsule
   length=5.9723
   radius=4.5972
   shapeName=dyn_sweatshirt_R_side_01_capsule
   center=<4.5549 -1.8357 3.0189>
   rotation=<rot -0.5011 179.398 -87.8296>
  }]
 dyn_sweatshirt_R_side_02 [{
   shapeType=Capsule
   length=6.7336
   radius=4.8733
   shapeName=dyn_sweatshirt_R_side_02_capsule
   center=<4.7005 -0.8654 2.9028>
   rotation=<rot 0 2.7576 -86.2313>
  }]
 dyn_sweatshirt_R_side_03 [{
   shapeType=Capsule
   length=4.7571
   radius=6.0936
   shapeName=dyn_sweatshirt_R_side_03_capsule
   center=<8.9567 0.9223 3.0195>
   rotation=<rot 0 4.8 -79.9926>
  }]
 foot_l [{shapeType=Capsule, length=13.0717, radius=7.4103, center=<-2.8995 8.0298 0.4863>, rotation=<rot 0 9.7641 -89.6864>}]
 foot_r [{shapeType=Capsule, length=13.0717, radius=7.4103, center=<2.8995 -8.0298 -0.4863>, rotation=<rot 0 9.7641 -89.6864>}]
 hand_l [{shapeType=Capsule, length=8.8952, radius=5.5155, center=<7.8386 -2.5524 -0.1927>, rotation=<rot 0 82.4177 -73.6551>}]
 hand_r [{shapeType=Capsule, length=8.8953, radius=5.5155, center=<-7.8386 2.5523 0.1927>, rotation=<rot 0 82.4146 -73.6555>}]
 kinematic_pelvis [{shapeType=Capsule, length=4.3341, radius=8.4873, shapeName=kinematic_pelvis_capsule, center=<0 -10.1038 0>, rotation=<rot 0 -90 0>}]
 lowerarm_l [{shapeType=Capsule, length=13.1538, radius=7.934, center=<10.072 0.6632 0.005>, rotation=<rot 0 -94.3788 -89.2403>}]
 lowerarm_r [{shapeType=Capsule, length=13.3492, radius=7.7009, center=<-9.8735 -0.4894 -0.0321>, rotation=<rot 0 -92.3828 -89.9067>}]
 pelvis [
  {shapeType=Capsule, length=8.7723, radius=11.0317, shapeName=pelvis_capsule, center=<0 -1.3809 0>}
  {shapeType=Capsule, length=13.0533, radius=10.0223, shapeName=pelvis_capsule, center=<-5.9628 -2.425 0>}
 ]
 spine_01 [{shapeType=Capsule, length=9.9057, radius=9.3196, shapeName=spine_01_capsule}]
 spine_02 [{shapeType=Capsule, length=8.8207, radius=9.3522, shapeName=spine_02_capsule}]
 spine_03 [{shapeType=Capsule, length=7.6988, radius=9.3791, shapeName=spine_03_capsule, rotation=<rot 0 -90 0>}]
 thigh_l [{shapeType=Capsule, length=32.6399, radius=8.1403, center=<-20.4785 -0.3312 -0.5777>, rotation=<rot 0 -89.7765 -89.2706>}]
 thigh_r [{shapeType=Capsule, length=22.2674, radius=8.2098, center=<20.774 0.5955 0.3358>, rotation=<rot 0.0003 -89.1386 -89.5682>}]
 upperarm_l [{shapeType=Capsule, length=14.636, radius=7.872, center=<11.4015 0.51 -1.3048>, rotation=<rot 0 92.2889 -83.358>}]
 upperarm_r [{shapeType=Capsule, length=14.7971, radius=8.3806, center=<-11.6674 -1.0398 1.3132>, rotation=<rot 0 90.9806 -82.2729>}]
}
constraints [
 {bone1Name=lowerarm_l, bone2Name=upperarm_l, swing1LimitDegrees=45, swing2LimitDegrees=45, twistLimitDegrees=45}
 {bone1Name=hand_l, bone2Name=lowerarm_l, swing1LimitDegrees=45, swing2LimitDegrees=45, twistLimitDegrees=45}
 {bone1Name=lowerarm_r, bone2Name=upperarm_r, swing1LimitDegrees=45, swing2LimitDegrees=45, twistLimitDegrees=45}
 {bone1Name=hand_r, bone2Name=lowerarm_r, swing1LimitDegrees=45, swing2LimitDegrees=45, twistLimitDegrees=45}
 {bone1Name=calf_l, bone2Name=thigh_l, swing1LimitDegrees=45, swing2LimitDegrees=45, twistLimitDegrees=45}
 {bone1Name=foot_l, bone2Name=calf_l, swing1LimitDegrees=45, swing2LimitDegrees=45, twistLimitDegrees=45}
 {bone1Name=calf_r, bone2Name=thigh_r, swing1LimitDegrees=45, swing2LimitDegrees=45, twistLimitDegrees=45}
 {bone1Name=foot_r, bone2Name=calf_r, swing1LimitDegrees=45, swing2LimitDegrees=45, twistLimitDegrees=45}
 {bone1Name=dyn_sweatshirt_L_back_03, bone2Name=dyn_sweatshirt_L_back_02, swing1LimitDegrees=5, swing2LimitDegrees=20, twistLimitDegrees=5}
 {bone1Name=dyn_sweatshirt_L_front_01, bone2Name=BASE_sweatshirt_L_front, swing1LimitDegrees=10, swing2LimitDegrees=80, twistLimitDegrees=3}
 {bone1Name=dyn_sweatshirt_L_front_02, bone2Name=dyn_sweatshirt_L_front_01, swing1LimitDegrees=10, swing2LimitDegrees=20, twistLimitDegrees=5}
 {bone1Name=dyn_sweatshirt_L_front_03, bone2Name=dyn_sweatshirt_L_front_02, swing1LimitDegrees=10, swing2LimitDegrees=20, twistLimitDegrees=5}
 {bone1Name=dyn_sweatshirt_L_side_01, bone2Name=BASE_sweatshirt_L_side, swing1LimitDegrees=10, swing2LimitDegrees=20, twistLimitDegrees=5}
 {bone1Name=dyn_sweatshirt_L_side_02, bone2Name=dyn_sweatshirt_L_side_01, swing1LimitDegrees=10, swing2LimitDegrees=10, twistLimitDegrees=5}
 {bone1Name=dyn_sweatshirt_L_side_03, bone2Name=dyn_sweatshirt_L_side_02, swing1LimitDegrees=20, swing2LimitDegrees=10, twistLimitDegrees=5}
 {bone1Name=dyn_sweatshirt_R_back_03, bone2Name=dyn_sweatshirt_R_back_02, swing1LimitDegrees=5, swing2LimitDegrees=20, twistLimitDegrees=5}
 {bone1Name=dyn_sweatshirt_R_front_01, bone2Name=BASE_sweatshirt_R_front, swing1LimitDegrees=10, swing2LimitDegrees=20, twistLimitDegrees=5}
 {bone1Name=dyn_sweatshirt_R_front_02, bone2Name=dyn_sweatshirt_R_front_01, swing1LimitDegrees=10, swing2LimitDegrees=20, twistLimitDegrees=5}
 {bone1Name=dyn_sweatshirt_R_front_03, bone2Name=dyn_sweatshirt_R_front_02, swing1LimitDegrees=10, swing2LimitDegrees=20, twistLimitDegrees=5}
 {bone1Name=dyn_sweatshirt_R_side_01, bone2Name=BASE_sweatshirt_R_side, swing1LimitDegrees=10, swing2LimitDegrees=20, twistLimitDegrees=5}
 {bone1Name=dyn_sweatshirt_R_side_02, bone2Name=dyn_sweatshirt_R_side_01, swing1LimitDegrees=10, swing2LimitDegrees=20, twistLimitDegrees=5}
 {bone1Name=dyn_sweatshirt_R_side_03, bone2Name=dyn_sweatshirt_R_side_02, swing1LimitDegrees=10, swing2LimitDegrees=20, twistLimitDegrees=5}
 {
  bone1Name=upperarm_l
  bone2Name=spine_03
  swing1LimitDegrees=45
  swing1Motion=Free
  swing2LimitDegrees=45
  swing2Motion=Free
  twistLimitDegrees=45
  twistMotion=Free
 }
 {
  bone1Name=upperarm_r
  bone2Name=spine_03
  swing1LimitDegrees=45
  swing1Motion=Free
  swing2LimitDegrees=45
  swing2Motion=Free
  twistLimitDegrees=45
  twistMotion=Free
 }
 {
  bone1Name=BASE_sweatshirt_R_front
  bone2Name=spine_03
  swing1LimitDegrees=45
  swing1Motion=Free
  swing2LimitDegrees=45
  swing2Motion=Free
  twistLimitDegrees=45
  twistMotion=Free
 }
 {
  bone1Name=BASE_sweatshirt_R_side
  bone2Name=spine_03
  swing1LimitDegrees=45
  swing1Motion=Free
  swing2LimitDegrees=45
  swing2Motion=Free
  twistLimitDegrees=45
  twistMotion=Free
 }
 {
  bone1Name=BASE_sweatshirt_R_back
  bone2Name=spine_03
  swing1LimitDegrees=45
  swing1Motion=Free
  swing2LimitDegrees=45
  swing2Motion=Free
  twistLimitDegrees=45
  twistMotion=Free
 }
 {
  bone1Name=BASE_sweatshirt_L_back
  bone2Name=spine_03
  swing1LimitDegrees=45
  swing1Motion=Free
  swing2LimitDegrees=45
  swing2Motion=Free
  twistLimitDegrees=45
  twistMotion=Free
 }
 {
  bone1Name=BASE_sweatshirt_L_side
  bone2Name=spine_03
  swing1LimitDegrees=45
  swing1Motion=Free
  swing2LimitDegrees=45
  swing2Motion=Free
  twistLimitDegrees=45
  twistMotion=Free
 }
 {
  bone1Name=BASE_sweatshirt_L_front
  bone2Name=spine_03
  swing1LimitDegrees=45
  swing1Motion=Free
  swing2LimitDegrees=45
  swing2Motion=Free
  twistLimitDegrees=45
  twistMotion=Free
 }
 {
  bone1Name=spine_02
  bone2Name=spine_01
  swing1LimitDegrees=45
  swing1Motion=Free
  swing2LimitDegrees=45
  swing2Motion=Free
  twistLimitDegrees=45
  twistMotion=Free
 }
 {
  bone1Name=spine_03
  bone2Name=spine_02
  swing1LimitDegrees=45
  swing1Motion=Free
  swing2LimitDegrees=45
  swing2Motion=Free
  twistLimitDegrees=45
  twistMotion=Free
 }
 {
  bone1Name=spine_01
  bone2Name=pelvis
  swing1LimitDegrees=45
  swing1Motion=Free
  swing2LimitDegrees=45
  swing2Motion=Free
  twistLimitDegrees=45
  twistMotion=Free
 }
 {
  bone1Name=thigh_l
  bone2Name=pelvis
  swing1LimitDegrees=45
  swing1Motion=Free
  swing2LimitDegrees=45
  swing2Motion=Free
  twistLimitDegrees=45
  twistMotion=Free
 }
 {
  bone1Name=thigh_r
  bone2Name=pelvis
  swing1LimitDegrees=45
  swing1Motion=Free
  swing2LimitDegrees=45
  swing2Motion=Free
  twistLimitDegrees=45
  twistMotion=Free
 }
 {
  bone1Name=dyn_sweatshirt_R_front_03
  bone2Name=dyn_sweatshirt_R_side_03
  swing1LimitDegrees=45
  swing1Motion=Free
  swing2LimitDegrees=45
  swing2Motion=Free
  twistLimitDegrees=45
  twistMotion=Free
 }
 {
  bone1Name=dyn_sweatshirt_L_front_03
  bone2Name=dyn_sweatshirt_L_side_03
  swing1LimitDegrees=45
  swing1Motion=Free
  swing2LimitDegrees=45
  swing2Motion=Free
  twistLimitDegrees=45
  twistMotion=Free
 }
 {
  bone1Name=kinematic_pelvis
  bone2Name=pelvis
  swing1LimitDegrees=45
  swing1Motion=Free
  swing2LimitDegrees=45
  swing2Motion=Free
  twistLimitDegrees=45
  twistMotion=Free
 }
 {
  bone1Name=dyn_sweatshirt_L_side_03
  bone2Name=dyn_sweatshirt_L_back_03
  swing1LimitDegrees=45
  swing1Motion=Free
  swing2LimitDegrees=45
  swing2Motion=Free
  twistLimitDegrees=45
  twistMotion=Free
 }
 {
  bone1Name=BASE_sweatshirt_M_back
  bone2Name=spine_03
  swing1LimitDegrees=45
  swing1Motion=Free
  swing2LimitDegrees=45
  swing2Motion=Free
  twistLimitDegrees=45
  twistMotion=Free
 }
 {bone1Name=dyn_sweatshirt_M_back_03, bone2Name=dyn_sweatshirt_M_back_02, swing1LimitDegrees=5, swing2LimitDegrees=25, twistLimitDegrees=0}
 {
  bone1Name=dyn_sweatshirt_L_back_03
  bone2Name=dyn_sweatshirt_M_back_03
  swing1LimitDegrees=45
  swing1Motion=Free
  swing2LimitDegrees=45
  swing2Motion=Free
  twistLimitDegrees=45
  twistMotion=Free
 }
 {bone1Name=dyn_sweatshirt_L_back_02, bone2Name=BASE_sweatshirt_L_back_02, swing1LimitDegrees=5, swing2LimitDegrees=20, twistLimitDegrees=5}
 {
  bone1Name=BASE_sweatshirt_L_back_02
  bone2Name=spine_02
  swing1LimitDegrees=45
  swing1Motion=Free
  swing2LimitDegrees=45
  swing2Motion=Free
  twistLimitDegrees=45
  twistMotion=Free
 }
 {bone1Name=dyn_sweatshirt_M_back_02, bone2Name=BASE_sweatshirt_M_back_02, swing1LimitDegrees=5, swing2LimitDegrees=25, twistLimitDegrees=0}
 {
  bone1Name=BASE_sweatshirt_M_back_02
  bone2Name=spine_02
  swing1LimitDegrees=45
  swing1Motion=Free
  swing2LimitDegrees=45
  swing2Motion=Free
  twistLimitDegrees=45
  twistMotion=Free
 }
 {bone1Name=dyn_sweatshirt_R_back_02, bone2Name=BASE_sweatshirt_R_back_02, swing1LimitDegrees=5, swing2LimitDegrees=20, twistLimitDegrees=5}
 {
  bone1Name=BASE_sweatshirt_R_back_02
  bone2Name=spine_02
  swing1LimitDegrees=45
  swing1Motion=Free
  swing2LimitDegrees=45
  swing2Motion=Free
  twistLimitDegrees=45
  twistMotion=Free
 }
]
