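# Post-process style AnimBlueprint for the female PillowMill Myriad character body, layering procedural rigs and physics onto the parent mesh's pose.
# Copied pose runs through a movement-state-driven Control Rig, a spine_02-space RigidBody simulation, then a post-sim Control Rig.
# Each update sets slide floats and a gravity override, and flags when the frontend idle montage plays for the post-sim rig.
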AnimBlueprint /BRCosmetics/Characters/Player/Female/Medium/Bodies/F_MED_PillowMill_Myriad/Meshes/F_MED_PillowMill_Myriad_AnimBP
target_skeleton=&/F_MED_PillowMill_Myriad_Skeleton
graphs [
 AnimGraph {nodes [
   n1 = Control Rig!(
    Source=Copy Pose From Mesh() {bUseAttachedParent=true}
    Alpha
    crouching:bool=Get bIsCrouching()
    DBNO:bool=Get bIsDBNO()
    baseball_sliding:double=Get baseball_sliding() -> (baseball_sliding:double)
    RBAN:bool=Get bRigidBodyEnabled()
    skydiving:bool=Get bIsSkyDiving()
   ) -> (Pose) {
    alphaBoolBlend {blendOption=Linear}
    controlRigAssetReference {blueprintRigClass=&/F_MED_PillowMill_Myriad_CtrlRig.F_MED_PillowMill_Myriad_CtrlRig_C}
   }
   n2 = RigidBody!(ComponentPose=Local To Component(LocalPose=@n1), bAlphaBoolEnabled:bool, OverrideWorldGravity:Vector="0,0,0", ExternalForce) -> (Pose) {
    alphaInputType=Bool
    overridePhysicsAsset=&/F_MED_PillowMill_Myriad_Physics
    simulationSpace=BaseBoneSpace
    alphaBoolBlend {blendInTime=0.2, blendOption=Linear, blendOutTime=0.2}
    baseBoneRef {boneName=spine_02}
    componentAppliedLinearAccClamp=<1200 1200 1200>
    componentLinearAccScale=<0.25 0.25 0.25>
    componentLinearVelScale=<0.3 0.3 0.3>
    simSpaceSettings {
     maxAngularAcceleration=100
     maxAngularVelocity=2
     maxLinearAcceleration=100
     maxLinearVelocity=160
     velocityScaleZ=0
     worldAlpha=0.75
     externalLinearDragV=<1 1 1>
    }
   }
   n3 = Control Rig!(
    Source=Component To Local(ComponentPose=@n2)
    Alpha
    frontend_anim_PillowMill_Female_Idle:double=Get frontend_anim_PillowMill_Female_Idle() -> (frontend_anim_PillowMill_Female_Idle:double)
   ) -> (Pose) {
    alphaBoolBlend {blendOption=Linear}
    controlRigAssetReference {blueprintRigClass=&/F_MED_PillowMill_Myriad_CtrlRig_postsim.F_MED_PillowMill_Myriad_CtrlRig_postsim_C}
   }
   Output Pose(Result=@n3) {layerGroup=None}
  ]}
 EventGraph {nodes [
   Event BlueprintUpdateAnimation()
   > Set knee_sliding(knee_sliding:double="To Float (Boolean)"(InBool=Get bIsKneeSliding())) -> (Output_Get:double)
   > Set baseball_sliding(baseball_sliding:double="To Float (Boolean)"(InBool=Get bIsBaseballSliding())) -> (Output_Get:double)
   > n1 = Gravity Override State Function DEPRECATED(CharacterPartAnimInstance=Self-Reference(), GameTimeInSeconds=GetGameTimeInSeconds(), GravityOverrideParams=Get Gravity Override Params())
   > Set Gravity Override(Gravity Override=@n1)
   n2 = GetAnimInstance(self=Get Mesh(self=TryGetFortPawn()))
   > n3 = Branch(Condition="Equal (Object)"(
     A=GetCurrentActiveMontage!(self:Anim Instance Object Reference=@n2) -> (ReturnValue:Anim Montage Object Reference)
     B=/Game/Animation/Game/MainPlayer/Menu/BR/CustomSkins/_s31/PillowMill_MasteryF/PillowMill_Female_Idle_Montage
    ))
   > Set frontend_anim_PillowMill_Female_Idle(frontend_anim_PillowMill_Female_Idle:double=1) -> (Output_Get:double)
   Set frontend_anim_PillowMill_Female_Idle(execute=@n3.else, frontend_anim_PillowMill_Female_Idle=0)
  ]}
]
variables [
 baseball_sliding {type=number}
 frontend_anim_PillowMill_Female_Idle {type=number}
 Gravity Override {type=Vector}
 Gravity Override Params {type=GravityOverrideParamsStruct}
 knee_sliding {type=number}
]
