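# Pre-simulation Control Rig for the PillowMill Myriad female character's face accessory, correcting right-side hair base bones as the head moves.
# A spherical pose reader measures the head bone's orientation relative to spine_05 and outputs a weight.
# That weight drives two chained transform modifiers: one offsets the front-side hair base, the other offsets and squashes the side hair base.
# Left, front and back hair, collar bones and the crouching variable are unused in the graph.
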ControlRigBlueprint /BRCosmetics/Characters/Player/Female/Medium/Bodies/F_MED_PillowMill_Myriad/Meshes/Parts/F_MED_PillowMill_Myriad_FaceAcc_CtrlRig__presim
graphs [
 RigVMFunctionLibrary {links [], nodes []}
 RigVMModel {nodes [
   RigUnit_BeginExecution = &RigUnit_BeginExecution()
   n1 = &RigUnit_SphericalPoseReader(
    ActiveRegionScaleFactors=(PositiveWidth=0.4,NegativeWidth=0.4,PositiveHeight=0.4,NegativeHeight=0.4)
    ActiveRegionSize=0.5
    Debug=(bDrawDebug=True,bDraw2D=false,bDrawLocalAxes=false,DebugScale=25,DebugSegments=20,DebugThickness=0.25)
    DriverAxis=(X=1,Y=0,Z=0)
    DriverItem=(Type=Bone,Name="head")
    FalloffRegionScaleFactors=(PositiveWidth=0.2,NegativeWidth=0.2,PositiveHeight=0.2,NegativeHeight=0.2)
    FalloffSize=1
    FlipWidthScaling=False
    OptionalParentItem=(Type=Bone,Name="spine_05")
    RotationOffset=(X=-30,Y=0,Z=0)
    ExecutePin=@RigUnit_BeginExecution.ExecutePin
   )
   > &RigUnit_ModifyTransforms(
    ItemToModify=((Item=(Type=Bone,Name="dyn_f_side_hair_BASE_r"),Transform=(Rotation=(X=0,Y=0,Z=0,W=1),Translation=(X=-1,Y=0,Z=0),Scale3D=(X=1,Y=1,Z=1))))
    Weight=@n1.OutputParam
   )
   > &RigUnit_ModifyTransforms(
    ItemToModify=((Item=(Type=Bone,Name="dyn_side_hair_BASE_r"),Transform=(Rotation=(X=0,Y=0,Z=0,W=1),Translation=(X=-3,Y=0,Z=0),Scale3D=(X=0.8,Y=1,Z=1))))
    Weight=@n1.OutputParam
   )
  ]}
]
hierarchy [
 root {loc=<0 0 0>, rot=<0 0 0 1>}
 pelvis {parent=root, loc=<0 -0.5301 98.0571>, rot=<0.0127 0.707 -0.0127 -0.707>}
 spine_01 {parent=pelvis, loc=<6.1208 -0.9328 0>, rot=<0 0 0.034 -0.9994>}
 spine_02 {parent=spine_01, loc=<6.5996 -0.2095 0>, rot=<0 0 0.0238 -0.9997>}
 spine_03 {parent=spine_02, loc=<6.4914 -0.1237 0>, rot=<0 0 0.0149 -0.9999>}
 spine_04 {parent=spine_03, loc=<6.9598 0.0794 0>, rot=<0 0 -0.0132 -0.9999>}
 spine_05 {parent=spine_04, loc=<6.9987 0 0>, rot=<0 0 -0.0015 -1>}
 neck_01 {parent=spine_05, loc=<11.3138 0.1529 0>, rot=<0 0 -0.2003 -0.9797>}
 neck_02 {parent=neck_01, loc=<4.5627 0 0>, rot=<0 0 0.0504 -0.9987>}
 head {parent=neck_02, loc=<5.3914 0 0>, rot=<0 0 0.115 -0.9934>}
 dyn_f_side_hair_BASE_r {parent=head, loc=<5.4076 3.8497 8.099>, rot=<-0.0043 0.0607 0.9974 -0.0381>}
 dyn_f_side_hair_01_r {parent=dyn_f_side_hair_BASE_r, loc=<0 0 0>, rot=<0 0 0 1>}
 dyn_f_side_hair_02_r {parent=dyn_f_side_hair_01_r, loc=<5.6269 0 0.1299>, rot=<0 0 0 1>}
 dyn_side_hair_BASE_r {parent=head, loc=<6.6341 -4.8021 7.6799>, rot=<0.0083 -0.2677 0.9628 -0.0374>}
 dyn_side_hair_01_r {parent=dyn_side_hair_BASE_r, loc=<0 0 0>, rot=<0 0 0 1>}
 dyn_side_hair_02_r {parent=dyn_side_hair_01_r, loc=<5.6269 0 0.1299>, rot=<0 0 0 1>}
 dyn_back_hair_BASE_m {parent=head, loc=<6.9667 -8.696 -0.6959>, rot=<0.0268 -0.7353 0.6766 -0.0274>}
 dyn_back_hair_01_m {parent=dyn_back_hair_BASE_m, loc=<0 0 0>, rot=<0 0 0 1>}
 dyn_back_hair_02_m {parent=dyn_back_hair_01_m, loc=<5.6269 0 0.1299>, rot=<0 0 0 1>}
 dyn_side_hair_BASE_l {parent=head, loc=<6.6527 -4.1934 -8.5049>, rot=<0.0362 -0.96 0.2773 -0.0126>}
 dyn_side_hair_01_l {parent=dyn_side_hair_BASE_l, loc=<0 0 0>, rot=<0 0 0 1>}
 dyn_side_hair_02_l {parent=dyn_side_hair_01_l, loc=<5.6269 0 0.1299>, rot=<0 0 0 1>}
 dyn_f_side_hair_BASE_l {parent=head, loc=<6.0639 3.462 -8.1782>, rot=<0.0383 -0.9931 -0.1106 0.0022>}
 dyn_f_side_hair_01_l {parent=dyn_f_side_hair_BASE_l, loc=<0 0 0>, rot=<0 0 0 1>}
 dyn_f_side_hair_02_l {parent=dyn_f_side_hair_01_l, loc=<5.6269 0 0.1299>, rot=<0 0 0 1>}
 dyn_front_hair_BASE_r {parent=head, loc=<9.1389 9.1154 3.6203>, rot=<-0.0018 -0.0053 0.9993 -0.0383>}
 dyn_front_hair_01_r {parent=dyn_front_hair_BASE_r, loc=<0 0 0>, rot=<0 0 0 1>}
 dyn_front_hair_BASE_l {parent=head, loc=<8.992 10.2258 -2.2819>, rot=<-0.0018 -0.0053 0.9993 -0.0383>}
 dyn_front_hair_01_l {parent=dyn_front_hair_BASE_l, loc=<0 0 0>, rot=<0 0 0 1>}
 collar_L {parent=spine_05, loc=<12.6252 -3.1947 -3.5252>, rot=<-0.2947 -0.7342 -0.154 -0.592>}
 collar_M {parent=spine_05, loc=<12.7771 -4.3168 -0.0139>, rot=<-0.1095 -0.6986 0.1095 -0.6986>}
 collar_R {parent=spine_05, loc=<12.5762 -3.372 3.6287>, rot=<0.1504 -0.6084 0.3226 -0.7093>}
]
variables [crouching {type=bool, default=False}]
