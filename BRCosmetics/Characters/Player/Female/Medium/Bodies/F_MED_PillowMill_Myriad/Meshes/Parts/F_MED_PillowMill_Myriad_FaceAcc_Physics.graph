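# Physics asset for the PillowMill Myriad female character's FaceAcc mesh part, defining bodies for dynamic hair motion.
# Short back, side, front-side and front hair chains hang from the head through tiny BASE capsules with free swing and twist.
# Links further down each chain have narrow swing limits and tighter twist limits.
# Head and collar capsules supply body volumes around the hair; the collar bodies have no constraints.
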PhysicsAsset /BRCosmetics/Characters/Player/Female/Medium/Bodies/F_MED_PillowMill_Myriad/Meshes/Parts/F_MED_PillowMill_Myriad_FaceAcc_Physics
bodies {
 collar_M [{
   shapeType=Capsule
   length=5.6756
   radius=3.3958
   shapeName=collar_R_capsule
   center=<-0.2247 -2.0323 -5.3033>
   rotation=<rot -0.0005 -90 89.1704>
  }]
 collar_R1 [{
   shapeType=Capsule
   length=9.2453
   radius=2.3918
   shapeName=collar_R1_capsule
   center=<-0.0039 1.1924 -1.2891>
   rotation=<rot -89.8265 179.9929 -179.9929>
  }]
 dyn_back_hair_01_m [{
   shapeType=Capsule
   length=7.8482
   radius=3.2164
   shapeName=dyn_back_hair_01_m_capsule
   center=<3.4958 -0.0504 0.11>
   rotation=<rot 88.4457 -168.6334 -80.2535>
  }]
 dyn_back_hair_02_m [{
   shapeType=Capsule
   length=6.404
   radius=1.636
   shapeName=dyn_back_hair_02_m_capsule
   center=<4.1491 0.0234 -2.0687>
   rotation=<rot 2.8454 -178.6506 -90.3089>
  }]
 dyn_back_hair_BASE_m [{shapeType=Capsule, length=0.505, radius=0.505, shapeName=dyn_back_hair_BASE_m_capsule, rotation=<rot 0 -90 0>}]
 dyn_f_side_hair_01_l [{
   shapeType=Capsule
   length=5.7794
   radius=1.5531
   shapeName=dyn_f_side_hair_01_l_capsule
   center=<3.7537 -0.4026 0.1529>
   rotation=<rot 8.4662 -17.9876 -78.2918>
  }]
 dyn_f_side_hair_01_r [{
   shapeType=Capsule
   length=4.3277
   radius=3.497
   shapeName=dyn_f_side_hair_01_r_capsule
   center=<4.0904 0.1519 1.7578>
   rotation=<rot 0 164.5601 -77.1965>
  }]
 dyn_f_side_hair_02_l [{
   shapeType=Capsule
   length=7.1958
   radius=1.6068
   shapeName=dyn_f_side_hair_02_l_capsule
   center=<3.8984 1.0562 -0.7831>
   rotation=<rot 0 -9.7895 -83.6189>
  }]
 dyn_f_side_hair_02_r [{
   shapeType=Capsule
   length=6.3966
   radius=1.9125
   shapeName=dyn_f_side_hair_02_r_capsule
   center=<3.3626 -0.9615 -1.5194>
   rotation=<rot 0 -150.1887 -62.7467>
  }]
 dyn_f_side_hair_BASE_l [{shapeType=Capsule, length=0.505, radius=0.505, shapeName=dyn_f_side_hair_BASE_l_capsule, rotation=<rot 0 -90 0>}]
 dyn_f_side_hair_BASE_r [{shapeType=Capsule, length=0.505, radius=0.505, shapeName=dyn_f_side_hair_BASE_r_capsule, rotation=<rot 0 -90 0>}]
 dyn_front_hair_01_l [{
   shapeType=Capsule
   length=4.8381
   radius=1.687
   shapeName=dyn_front_hair_01_l_capsule
   center=<2.5359 -0.2983 -0.0043>
   rotation=<rot -31.3421 -90 14.5704>
  }]
 dyn_front_hair_01_r [{
   shapeType=Capsule
   length=5.0862
   radius=1.7005
   shapeName=dyn_front_hair_01_r_capsule
   center=<3.3563 -0.7739 -0.2122>
   rotation=<rot 21.0565 -90 6.5832>
  }]
 dyn_front_hair_BASE_l [{shapeType=Capsule, length=0.505, radius=0.505, shapeName=dyn_front_hair_BASE_l_capsule, rotation=<rot 0 -90 0>}]
 dyn_front_hair_BASE_r [{shapeType=Capsule, length=0.505, radius=0.505, shapeName=dyn_front_hair_BASE_r_capsule, rotation=<rot 0 -90 0>}]
 dyn_side_hair_01_l [{
   shapeType=Capsule
   length=6.9981
   radius=2.7961
   shapeName=dyn_side_hair_01_l_capsule
   center=<4.249 -0.3989 -0.061>
   rotation=<rot 0 -14.0663 -86.4168>
  }]
 dyn_side_hair_01_r [{
   shapeType=Capsule
   length=4.2212
   radius=2.2929
   shapeName=dyn_side_hair_01_r_capsule
   center=<4.1031 -0.1079 1.0599>
   rotation=<rot 0 178.2344 -83.3604>
  }]
 dyn_side_hair_02_l [{
   shapeType=Capsule
   length=5.0881
   radius=2.3561
   shapeName=dyn_side_hair_02_l_capsule
   center=<2.8618 1.4175 -0.7925>
   rotation=<rot -8.8009 171.8037 -120.1738>
  }]
 dyn_side_hair_02_r [
  {
   shapeType=Capsule
   length=8.1764
   radius=2.0996
   shapeName=dyn_side_hair_02_r_capsule
   center=<3.9606 -1.446 -2.6547>
   rotation=<rot 0 -158.1626 -66.4446>
  }
  {
   shapeType=Capsule
   length=2.2483
   radius=0.7796
   shapeName=dyn_side_hair_02_r_capsule
   center=<4.8903 3.7626 -3.186>
   rotation=<rot -21.0465 -54.1385 -66.1079>
  }
 ]
 dyn_side_hair_BASE_l [{shapeType=Capsule, length=0.505, radius=0.505, shapeName=dyn_side_hair_BASE_l_capsule, rotation=<rot 0 -90 0>}]
 dyn_side_hair_BASE_r [{shapeType=Capsule, length=0.505, radius=0.505, shapeName=dyn_side_hair_BASE_r_capsule, rotation=<rot 0 -90 0>}]
 head [
  {shapeType=Capsule, length=6.7534, radius=6.3369, center=<3.1734 2.6019 0.0541>, rotation=<rot 0.5016 -90.415 -88.3249>}
  {
   shapeType=Capsule
   length=3.3889
   radius=6.0811
   shapeName=head_capsule
   center=<1.6133 -0.0375 -0.0239>
   rotation=<rot -89.1519 179.9981 -180>
  }
 ]
}
constraints [
 {
  bone1Name=dyn_back_hair_BASE_m
  bone2Name=head
  swing1LimitDegrees=45
  swing1Motion=Free
  swing2LimitDegrees=45
  swing2Motion=Free
  twistLimitDegrees=45
  twistMotion=Free
 }
 {bone1Name=dyn_back_hair_01_m, bone2Name=dyn_back_hair_BASE_m, swing1LimitDegrees=15, swing2LimitDegrees=10, twistLimitDegrees=1}
 {bone1Name=dyn_back_hair_02_m, bone2Name=dyn_back_hair_01_m, swing1LimitDegrees=15, swing2LimitDegrees=10, twistLimitDegrees=1}
 {
  bone1Name=dyn_f_side_hair_BASE_l
  bone2Name=head
  swing1LimitDegrees=45
  swing1Motion=Free
  swing2LimitDegrees=45
  swing2Motion=Free
  twistLimitDegrees=45
  twistMotion=Free
 }
 {bone1Name=dyn_f_side_hair_01_l, bone2Name=dyn_f_side_hair_BASE_l, swing1LimitDegrees=20, swing2LimitDegrees=20, twistLimitDegrees=5}
 {bone1Name=dyn_f_side_hair_02_l, bone2Name=dyn_f_side_hair_01_l, swing1LimitDegrees=20, swing2LimitDegrees=20, twistLimitDegrees=5}
 {
  bone1Name=dyn_f_side_hair_BASE_r
  bone2Name=head
  swing1LimitDegrees=45
  swing1Motion=Free
  swing2LimitDegrees=45
  swing2Motion=Free
  twistLimitDegrees=45
  twistMotion=Free
 }
 {bone1Name=dyn_f_side_hair_01_r, bone2Name=dyn_f_side_hair_BASE_r, swing1LimitDegrees=20, swing2LimitDegrees=20, twistLimitDegrees=5}
 {bone1Name=dyn_f_side_hair_02_r, bone2Name=dyn_f_side_hair_01_r, swing1LimitDegrees=20, swing2LimitDegrees=20, twistLimitDegrees=5}
 {
  bone1Name=dyn_front_hair_BASE_l
  bone2Name=head
  swing1LimitDegrees=45
  swing1Motion=Free
  swing2LimitDegrees=45
  swing2Motion=Free
  twistLimitDegrees=45
  twistMotion=Free
 }
 {bone1Name=dyn_front_hair_01_l, bone2Name=dyn_front_hair_BASE_l, swing1LimitDegrees=20, swing2LimitDegrees=20, twistLimitDegrees=5}
 {
  bone1Name=dyn_front_hair_BASE_r
  bone2Name=head
  swing1LimitDegrees=45
  swing1Motion=Free
  swing2LimitDegrees=45
  swing2Motion=Free
  twistLimitDegrees=45
  twistMotion=Free
 }
 {bone1Name=dyn_front_hair_01_r, bone2Name=dyn_front_hair_BASE_r, swing1LimitDegrees=20, swing2LimitDegrees=20, twistLimitDegrees=5}
 {
  bone1Name=dyn_side_hair_BASE_l
  bone2Name=head
  swing1LimitDegrees=45
  swing1Motion=Free
  swing2LimitDegrees=45
  swing2Motion=Free
  twistLimitDegrees=45
  twistMotion=Free
 }
 {bone1Name=dyn_side_hair_01_l, bone2Name=dyn_side_hair_BASE_l, swing1LimitDegrees=10, swing2LimitDegrees=20, twistLimitDegrees=1}
 {bone1Name=dyn_side_hair_02_l, bone2Name=dyn_side_hair_01_l, swing1LimitDegrees=10, swing2LimitDegrees=20, twistLimitDegrees=1}
 {
  bone1Name=dyn_side_hair_BASE_r
  bone2Name=head
  swing1LimitDegrees=45
  swing1Motion=Free
  swing2LimitDegrees=45
  swing2Motion=Free
  twistLimitDegrees=45
  twistMotion=Free
 }
 {bone1Name=dyn_side_hair_01_r, bone2Name=dyn_side_hair_BASE_r, swing1LimitDegrees=10, swing2LimitDegrees=20, twistLimitDegrees=1}
 {bone1Name=dyn_side_hair_02_r, bone2Name=dyn_side_hair_01_r, swing1LimitDegrees=10, swing2LimitDegrees=20, twistLimitDegrees=1}
]
